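# Sound cue for the Reality Forest biome's Nexus Core underwater "above" ambience, classed as outdoor Fortnite Athena ambient audio.
# A looping wave player plays a single underwater loop through a modulator that applies a fixed volume boost with no pitch variation.
# The cue's volume multiplier trims the output, and a shared Nexus Core underwater attenuation asset handles distance falloff.
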SoundCue /Game/Sounds/Artemis/Biomes/Reality_Forest/AMB_Nexus_Core_Underwater_Above_Cue
root=Modulator_0
nodes [
 {id=Modulator_0, children [WavePlayer_1], properties {pitchMax=1, pitchMin=1, volumeMax=1.5, volumeMin=1.5}}
 {
  id=WavePlayer_1
  properties {bLooping=true, soundWaveAssetPtr=&/AMB_Nexus_Core_Underwater_Above_Loop_01}
 }
]
properties {
 attenuationSettings=&/AMB_Nexus_Core_Underwater_Att
 maxDistance=6000
 soundClassObject=&Fort_Ambient_Outdoors_Athena
 volumeMultiplier=0.752
}
