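# Looping ambient haptics cue for below the water surface at the Nexus Core in the Reality Forest biome.
# A single wave player loops a rumble-trigger variant of a below-water ambience loop, with boosted volume and no mixing or randomization nodes.
# It uses the Nexus Core underwater attenuation and an in-game SFX class excluded from effects, with base submix output disabled.
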SoundCue /Game/Sounds/Artemis/Biomes/Reality_Forest/AMB_Nexus_Core_Underwater_Below_Haptics_Cue
root=WavePlayer_2
nodes [{
  id=WavePlayer_2
  properties {bLooping=true, soundWaveAssetPtr=&/AMB_S21_ZP_Water_Below_Loop_01_Rumble_Trigger}
 }]
properties {
 attenuationSettings=&/AMB_Nexus_Core_Underwater_Att
 bEnableBaseSubmix=false
 maxDistance=6000
 soundClassObject=/Game/Sounds/FortSoundClassesAndModes/Fort_SFX_InGame_ExcludeFromFX
 volumeMultiplier=1.5
}
