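; Underwater audio and controller-rumble actor for the Reality Forest Nexus Core (ZP), switched by an audio gameplay volume condition that tests whether the camera is underwater.
; Toggling fades between above and below audio components, starting the below loop at a random time, while a timeline drives the above sound's "Crossfade" parameter.
; While submerged, the haptics component's envelope, normalized against its running peak and scaled by listener distance, drives the local player's dynamic force feedback.
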(event OnToggledOn(AudioGameplayVolumeProxy)
  (Development|PrintString "Underwater ZP: True")
  (CallFunction|StartRumble)
  (bind (_direction _crossfade) (Timeline|Crossfade Play)
    (:Update
      (Audio|Parameter|SetFloatParameter (Variables|Default|GetAudioAbove) "Crossfade" _crossfade :!class /Script/AudioExtensions.AudioParameterControllerInterface))))

(event OnToggledOff(AudioGameplayVolumeProxy)
  (Development|PrintString "Underwater ZP: False")
  (CallFunction|StopRumble)
  (Timeline|Crossfade Reverse))

(event EventBeginPlay
  (Development|PrintString "Underwater ZP: False")
  (CallFunction|StopRumble)
  (Timeline|Crossfade Reverse))

(event Custom|Start%20Rumble
  (Variables|Default|SetbRumbleIsActive true)
  (bind _output_get (Variables|Default|SetRandomStartTime (Math|Random|RandomFloatinRange 0.0 (Variables|/Script/Engine.SoundBase|GetDuration (Variables|Default|GetBelowLoopWave)))))
  (Audio|Components|Audio|FadeIn (Variables|Default|GetAudioBelow) 0.25 1.0 _output_get "Linear" :!class /Script/Engine.AudioComponent)
  (Audio|Components|Audio|FadeOut (Variables|Default|GetAudioAbove) 0.25 0.0 "Linear" :!class /Script/Engine.AudioComponent)
  (Game|Feedback|PlayDynamicForceFeedback (Variables|Default|GetLocalPC) (* (Math|Float|MapRangeClamped (stale OnAudioMultiEnvelopeValue(AudioBelowHaptics) _output_get) 0.0 (* (Variables|/Script/Engine.SoundBase|GetMaxDistance (Variables|/Script/Engine.AudioComponent|GetSound (Variables|Default|GetAudioBelow))) 0.75) 1.0) (Math|Float|MapRangeClamped (Variables|Default|GetCurrentEnvValue) 0.0 (Variables|Default|GetMaxEnvValue) 0.0 (Variables|Default|GetMaxRumbleValue))) -1 (< (Variables|Default|GetDistanceFromZP) (* (Variables|/Script/Engine.SoundBase|GetMaxDistance (Variables|/Script/Engine.AudioComponent|GetSound (Variables|Default|GetAudioBelow))) 0.5)) true (< (Variables|Default|GetDistanceFromZP) (* (Variables|/Script/Engine.SoundBase|GetMaxDistance (Variables|/Script/Engine.AudioComponent|GetSound (Variables|Default|GetAudioBelow))) 0.5)) true))

(event Custom|Stop%20Rumble
  (Variables|Default|SetbRumbleIsActive false)
  (Audio|Components|Audio|FadeOut (Variables|Default|GetAudioBelow) 0.25 0.0 "Linear" :!class /Script/Engine.AudioComponent)
  (Audio|Components|Audio|FadeIn (Variables|Default|GetAudioAbove) 0.25 1.0 0.0 "Linear" :!class /Script/Engine.AudioComponent)
  (Game|Feedback|PlayDynamicForceFeedback (Variables|Default|GetLocalPC) (* (Math|Float|MapRangeClamped (stale OnAudioMultiEnvelopeValue(AudioBelowHaptics) _output_get) 0.0 (* (Variables|/Script/Engine.SoundBase|GetMaxDistance (Variables|/Script/Engine.AudioComponent|GetSound (Variables|Default|GetAudioBelow))) 0.75) 1.0) (Math|Float|MapRangeClamped (Variables|Default|GetCurrentEnvValue) 0.0 (Variables|Default|GetMaxEnvValue) 0.0 (Variables|Default|GetMaxRumbleValue))) -1 (< (Variables|Default|GetDistanceFromZP) (* (Variables|/Script/Engine.SoundBase|GetMaxDistance (Variables|/Script/Engine.AudioComponent|GetSound (Variables|Default|GetAudioBelow))) 0.5)) true (< (Variables|Default|GetDistanceFromZP) (* (Variables|/Script/Engine.SoundBase|GetMaxDistance (Variables|/Script/Engine.AudioComponent|GetSound (Variables|Default|GetAudioBelow))) 0.5)) true))

(event OnAudioMultiEnvelopeValue(AudioBelowHaptics)
  (params (AverageEnvelopeValue :type Float%20%28single-precision%29) (MaxEnvelope :type Float%20%28single-precision%29) (NumWaveInstances :type Integer))
  (bind _returnvalue (Transformation|GetActorLocation))
  (bind _audiobelow (Variables|Default|GetAudioBelow))
  (bind _sound (Variables|/Script/Engine.AudioComponent|GetSound _audiobelow))
  (bind _maxdistance (Variables|/Script/Engine.SoundBase|GetMaxDistance _sound))
  (bind _returnvalue_2 (* _maxdistance 0.75))
  (if (Variables|Default|GetbRumbleIsActive)
    (Variables|Default|SetCurrentEnvValue MaxEnvelope)
    (if (> (Variables|Default|GetCurrentEnvValue) (Variables|Default|GetMaxEnvValue))
      (Variables|Default|SetMaxEnvValue (Variables|Default|GetCurrentEnvValue)))
    (bind _listenerposition (Audio|GetClosestListenerLocation _returnvalue _returnvalue_2 true))
    (bind _output_get (Variables|Default|SetDistanceFromZP (Math|Vector|Distance(Vector) _listenerposition _returnvalue)))
    (CallFunction|AssignLocalPC)
    (Game|Feedback|PlayDynamicForceFeedback (Variables|Default|GetLocalPC) (* (Math|Float|MapRangeClamped _output_get 0.0 _returnvalue_2 1.0) (Math|Float|MapRangeClamped (Variables|Default|GetCurrentEnvValue) 0.0 (Variables|Default|GetMaxEnvValue) 0.0 (Variables|Default|GetMaxRumbleValue))) -1 (< (Variables|Default|GetDistanceFromZP) (* (Variables|/Script/Engine.SoundBase|GetMaxDistance (Variables|/Script/Engine.AudioComponent|GetSound (Variables|Default|GetAudioBelow))) 0.5)) true (< (Variables|Default|GetDistanceFromZP) (* (Variables|/Script/Engine.SoundBase|GetMaxDistance (Variables|/Script/Engine.AudioComponent|GetSound (Variables|Default|GetAudioBelow))) 0.5)) true)))

(fn AssignLocalPC
  (Variables|Default|GetLocalPC
    (:then)
    (:else
      (Variables|Default|SetLocalPC (Fortnite|GetFirstLocalFortPlayerController true true)))))

(fn ConditionMet
  (return (> (Water|GetCameraUnderwaterDepth (WorldSubsystems|GetWaterSubsystem)) 0)))

(fn ConditionMet_Position
  (params (Position :type Vector%20%28by%20ref%29 :default "0, 0, 0"))
  (return false))

(fn ConstructionScript)
